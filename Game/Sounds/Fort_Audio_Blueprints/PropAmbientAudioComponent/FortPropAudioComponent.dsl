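; Prop audio component that plays an attached ambient sound loop on its owning building actor, optionally registered as a music event.
; On BeginPlay, where cosmetic events can run, it fades the loop in, with a setting to skip Switch and mobile.
; Registered loops leave the music event subsystem while virtualized and rejoin when audible again.
; The owner's death-effects delegate fades the loop out and unregisters it.
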(event EventBeginPlay
  (Networking|CanExecuteCosmeticEvents
    (:True
      (bind _asbuilding_actor (Utilities|Casting|CastToBuildingActor (Components|GetOwner)))
      (CallFunction|PlayAudio)
      (EventDispatchers|AssignOnDeathPlayEffectsDelegate _asbuilding_actor (EventDispatchers|CreateEvent :!event StopAudio)))))

(event Custom|Virtualization%20Changed
  (params (bIsVirtualized :type Boolean))
  (if bIsVirtualized
    (bind _loopaudiocomponent (Variables|Default|GetLoopAudioComponent)
      (:then
        (Audio|RemoveEvent (GameInstanceSubsystems|GetMusicEventSubsystem) _loopaudiocomponent))
      (:else))
    (else
      (bind _loopaudiocomponent_2 (Variables|Default|GetLoopAudioComponent)
        (:then
          (Audio|AddEventbyComponent (GameInstanceSubsystems|GetMusicEventSubsystem) _loopaudiocomponent_2 (Variables|Settings|GetMusicEventTag) "(GameplayTags=)" 0 (Variables|Settings|GetMusicGroupName)))
        (:else)))))

(fn PlayAudio
  (if (select (Variables|Settings|GetPlayOnMobileSwitch) true (not (or (TouchPlatform|IsNintendoSwitchorNintendoSwitchPreview) (Fortnite|IsMobilePlatform))))
    (bind _returnvalue (Audio|SpawnSoundAttached :Sound (Variables|Settings|GetSoundLoop) :AttachToComponent self :AttachPointName "None" :LocationType "KeepRelativeOffset" :bStopWhenAttachedToDestroyed true :VolumeMultiplier (Variables|Settings|GetVolumeMultiplier) :PitchMultiplier (Variables|Settings|GetPitchMultiplier) :StartTime 0.0 :bAutoDestroy true))
    (Audio|Components|Audio|FadeIn _returnvalue (Variables|Settings|GetFadeInTime) 1.0 0.0 "Linear" :!class /Script/Engine.AudioComponent)
    (Variables|Default|SetLoopAudioComponent _returnvalue)
    (if (Variables|Settings|GetbRegisterMusic)
      (bind _loopaudiocomponent (Variables|Default|GetLoopAudioComponent)
        (:then
          (Audio|AddEventbyComponent (GameInstanceSubsystems|GetMusicEventSubsystem) _loopaudiocomponent (Variables|Settings|GetMusicEventTag) "(GameplayTags=)" 0 (Variables|Settings|GetMusicGroupName))
          (EventDispatchers|BindEventtoOnAudioVirtualizationChanged (Variables|Default|GetLoopAudioComponent) (EventDispatchers|CreateEvent :!event Virtualization%20Changed))
          (CallFunction|VirtualizationChanged (Audio|Components|Audio|IsVirtualized (Variables|Default|GetLoopAudioComponent))))
        (:else)))))

(fn StopAudio
  (params (Damage :type Float%20%28single-precision%29) (DamageTags :type Gameplay%20Tag%20Container%20Structure%20%28by%20ref%29) (Momentum :type Vector) (HitInfo :type Hit%20Result%20Structure%20%28by%20ref%29) (InstigatedBy :type Fort%20Pawn%20Object%20Reference) (DamageCauser :type Actor%20Object%20Reference) (EffectContext :type Gameplay%20Effect%20Context%20Handle%20Structure))
  (bind _loopaudiocomponent (Variables|Default|GetLoopAudioComponent)
    (:then
      (EventDispatchers|UnbindallEventsfromOnAudioVirtualizationChanged _loopaudiocomponent)
      (Audio|Components|Audio|FadeOut (Variables|Default|GetLoopAudioComponent) (Variables|Settings|GetFadeOutTime) 0.0 "Linear" :!class /Script/Engine.AudioComponent)
      (if (Variables|Settings|GetbRegisterMusic)
        (Audio|RemoveEvent (GameInstanceSubsystems|GetMusicEventSubsystem) (Variables|Default|GetLoopAudioComponent))
        (return)
        (else
          (return))))
    (:else
      (return))))
